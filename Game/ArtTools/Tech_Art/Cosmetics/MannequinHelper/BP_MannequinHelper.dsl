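; BP_MannequinHelper is a tech-art helper actor that swaps its skeletal meshes to debug materials for inspecting textures, channels, normals, vertex color and UVs.
; InitMIDs caches original and debug dynamic material instances per mesh in MaterialMap.
; The construction script applies or restores these through InitAllMaterials, finding textures by parameter names chosen per debug type.
; UV preview passes a per-material island index and enlarges mesh bounds.
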(event EventBeginPlay)

(event Collision|EventActorBeginOverlap
  (params (OtherActor :type Actor%20Object%20Reference)))

(event EventTick
  (params (DeltaSeconds :type Float%20%28single-precision%29)))

(fn InitAllMaterials
  (bind _array (Utilities|Array|MakeArray "Mask" "M"))
  (bind _array_2 (Utilities|Array|MakeArray "TechArtMask" "SkinFX_Mask"))
  (bind _array_3 (Utilities|Array|MakeArray "SSC_Texture_Far"))
  (bind _array_4 (Utilities|Array|MakeArray "SRM" "SpecularMasks"))
  (bind _array_5 (Utilities|Array|MakeArray "Normals"))
  (bind _array_6 (Utilities|Array|MakeArray "Diffuse"))
  (bind _array_7 (Utilities|Array|MakeArray "LitDiffuse"))
  (bind _array_8 (Utilities|Array|MakeArray "ShadedDiffuse"))
  (bind _array_9 (Utilities|Array|MakeArray "DistanceField_InkLines"))
  (bind _array_10 (Utilities|Array|MakeArray "SSC_Texture"))
  (bind _debugmaterialtype (Variables|Default|GetDebugMaterialType))
  (bind _returnvalue (Utilities|Select :Index _debugmaterialtype :NewEnumerator1 _array_6 :NewEnumerator3 _array_4 :NewEnumerator4 _array_5 :NewEnumerator7 _array_5 :NewEnumerator5 _array :NewEnumerator6 _array_2 :NewEnumerator11 _array_7 :NewEnumerator12 _array_8 :NewEnumerator13 _array_9 :NewEnumerator14 _array_10 :NewEnumerator15 _array_3))
  (if (Variables|Default|GetShowDebugMaterial)
    (switch Utilities|FlowControl|Switch|SwitchonMannequinMaterialTypes (Variables|Default|GetDebugMaterialType)
      (:NewEnumerator1
        _)
      (:NewEnumerator3
        (CallFunction|InitDebugMaterial _returnvalue false false false false false true))
      (:NewEnumerator4
        (CallFunction|InitDebugMaterial _returnvalue true false false false false true))
      (:NewEnumerator7
        (CallFunction|InitDebugMaterial _returnvalue true true false false false true))
      (:NewEnumerator5
        _)
      (:NewEnumerator6
        _)
      (:NewEnumerator8
        (CallFunction|InitDebugMaterial :IsNormal false :WSNormal false :VertexColor true :UV0 false :UV1 false :DisplaySRGB true))
      (:NewEnumerator9
        (CallFunction|InitDebugMaterial :IsNormal false :WSNormal false :VertexColor false :UV0 true :UV1 false :DisplaySRGB true))
      (:NewEnumerator10
        (CallFunction|InitDebugMaterial :IsNormal false :WSNormal false :VertexColor false :UV0 false :UV1 true :DisplaySRGB true))
      (:NewEnumerator11
        _)
      (:NewEnumerator12
        _)
      (:NewEnumerator13
        _)
      (:NewEnumerator14
        _)
      (:NewEnumerator15
        _))
    (CallFunction|InitDebugMaterial _returnvalue)
    (else
      (Utilities|Array|Clear (Variables|Default|GetDebugCurrentTextures))
      (for _array_element (Actor|GetComponentsByClass "/Script/Engine.SkeletalMeshComponent")
        (bind _value (Utilities|Map|Find (Variables|Hidden|GetMaterialMap) _array_element))
        (bind _value_mids_3_6ff767fe41164d399b8e9a9d765c88d3 (Utilities|Struct|BreakMannequinHelperMIDData _value))
        (for (_array_element_2 _array_index) _value_mids_3_6ff767fe41164d399b8e9a9d765c88d3
          (Rendering|Material|SetMaterial _array_element _array_index _array_element_2))))))

(fn InitDebugMaterial
  (params (ParamNames :type Array%20of%20Names) (IsNormal :type Boolean) (WSNormal :type Boolean) (VertexColor :type Boolean) (UV0 :type Boolean) (UV1 :type Boolean) (DisplaySRGB :type Boolean))
  (bind _materialmap (Variables|Hidden|GetMaterialMap))
  (LocalVariables|SetL_ParamNames ParamNames)
  (LocalVariables|SetL_IsNormal IsNormal)
  (LocalVariables|SetL_WSNormal WSNormal)
  (LocalVariables|SetL_VertexColor VertexColor)
  (LocalVariables|SetL_UV0 UV0)
  (LocalVariables|SetL_UV1 UV1)
  (LocalVariables|SetL_DisplaySRGB DisplaySRGB)
  (LocalVariables|SetL_DrawIndex 0)
  (Utilities|Array|Clear (Variables|Default|GetDebugCurrentTextures))
  (bind _keys (Utilities|Map|Keys (Variables|Hidden|GetMaterialMap)))
  (for _array_element _keys
    (LocalVariables|SetL_SkelMeshComponent _array_element)
    (bind _value (Utilities|Map|Find _materialmap _array_element))
    (bind (_value_mids_3_6ff767fe41164d399b8e9a9d765c88d3 _value_debugmids_5_ee2cbf9841eff7ba62c7a3b272b05e8d) (Utilities|Struct|BreakMannequinHelperMIDData _value))
    (for (_array_element_2 _array_index) _value_debugmids_5_ee2cbf9841eff7ba62c7a3b272b05e8d
      (bind _output_get (LocalVariables|SetL_DebugMID _array_element_2))
      (LocalVariables|SetL_MIDIndex _array_index)
      (if (Utilities|Array|IsNotEmpty (LocalVariables|GetL_ParamNames))
        (CallFunction|InitDebugMaterialTextures (Utilities|Array|Get(acopy) _value_mids_3_6ff767fe41164d399b8e9a9d765c88d3 _array_index) _output_get (LocalVariables|GetL_ParamNames) (LocalVariables|GetL_IsNormal))
        (else
          (Utilities|Array|Clear (Variables|Default|GetDebugCurrentTextures))))
      (CallFunction|InitDebugMaterialParams (LocalVariables|GetL_DebugMID) (LocalVariables|GetL_IsNormal) (LocalVariables|GetL_WSNormal) (LocalVariables|GetL_VertexColor) (LocalVariables|GetL_UV0) (LocalVariables|GetL_UV1) (LocalVariables|GetL_DisplaySRGB) (LocalVariables|GetL_DrawIndex))
      (Rendering|Material|SetMaterial (LocalVariables|GetL_SkelMeshComponent) (LocalVariables|GetL_MIDIndex) (LocalVariables|GetL_DebugMID))
      (Math|Integer|IncrementInt (LocalVariables|GetL_DrawIndex)))))

(fn InitDebugMaterialParams
  (params (DebugMID :type Material%20Instance%20Dynamic%20Object%20Reference) (IsNormal :type Boolean) (WorldSpaceNormal :type Boolean) (VertexColor :type Boolean) (UV0 :type Boolean) (UV1 :type Boolean) (DisplaySRGB :type Boolean) (DrawIndex :type Integer))
  (bind _l_debugmid (LocalVariables|GetL_DebugMID))
  (bind _output_get (LocalVariables|SetL_DebugMID DebugMID))
  (Rendering|Material|SetScalarParameterValue _output_get "IsNormal" (Math|Conversions|ToFloat(Boolean) IsNormal) :!class /Script/Engine.MaterialInstanceDynamic)
  (Rendering|Material|SetScalarParameterValue (LocalVariables|GetL_DebugMID) "WorldSpaceNormal" (Math|Conversions|ToFloat(Boolean) WorldSpaceNormal) :!class /Script/Engine.MaterialInstanceDynamic)
  (Rendering|Material|SetScalarParameterValue (LocalVariables|GetL_DebugMID) "VertexColor" (Math|Conversions|ToFloat(Boolean) VertexColor) :!class /Script/Engine.MaterialInstanceDynamic)
  (Rendering|Material|SetScalarParameterValue (LocalVariables|GetL_DebugMID) "UV0" (Math|Conversions|ToFloat(Boolean) UV0) :!class /Script/Engine.MaterialInstanceDynamic)
  (Rendering|Material|SetScalarParameterValue (LocalVariables|GetL_DebugMID) "UV1" (Math|Conversions|ToFloat(Boolean) UV1) :!class /Script/Engine.MaterialInstanceDynamic)
  (Rendering|Material|SetScalarParameterValue (LocalVariables|GetL_DebugMID) "PreviewUVs" (Math|Conversions|ToFloat(Boolean) (Variables|Default|GetDebugPreviewUvs)) :!class /Script/Engine.MaterialInstanceDynamic)
  (Rendering|Material|SetScalarParameterValue (LocalVariables|GetL_DebugMID) "PreviewUVIslandSize" (Variables|Default|GetDebugPreviewIslandSize) :!class /Script/Engine.MaterialInstanceDynamic)
  (Rendering|Material|SetScalarParameterValue (LocalVariables|GetL_DebugMID) "PreviewUVIslandIndex" (Math|Conversions|ToFloat(Integer) DrawIndex) :!class /Script/Engine.MaterialInstanceDynamic)
  (Rendering|Material|SetScalarParameterValue (LocalVariables|GetL_DebugMID) "DisplaySRGB" (Math|Conversions|ToFloat(Boolean) DisplaySRGB) :!class /Script/Engine.MaterialInstanceDynamic)
  (Rendering|Material|SetScalarParameterValue (LocalVariables|GetL_DebugMID) "PreviewUVChannel" (Math|Conversions|ToFloat(Integer) (Variables|Default|GetDebugPreviewUVChannel)) :!class /Script/Engine.MaterialInstanceDynamic)
  (Rendering|Material|SetVectorParameterValue (LocalVariables|GetL_DebugMID) "Channel" (Utilities|Select :Index (Variables|Default|GetDebugMaterialChannel) :NewEnumerator0 "(R=1.000000,G=0.000000,B=0.000000,A=0.000000)" :NewEnumerator1 "(R=1.000000,G=0.000000,B=0.000000,A=0.000000)" :NewEnumerator2 "(R=0.000000,G=1.000000,B=0.000000,A=0.000000)" :NewEnumerator3 "(R=0.000000,G=0.000000,B=1.000000,A=0.000000)" :NewEnumerator4 "(R=0.000000,G=0.000000,B=0.000000,A=1.000000)") :!class /Script/Engine.MaterialInstanceDynamic)
  (switch Utilities|FlowControl|Switch|SwitchonMannequinHelperChannels (Variables|Default|GetDebugMaterialChannel)
    (:NewEnumerator0
      (Rendering|Material|SetScalarParameterValue _l_debugmid "PreviewChannel" 0.0 :!class /Script/Engine.MaterialInstanceDynamic))
    (:NewEnumerator1
      (Rendering|Material|SetScalarParameterValue _l_debugmid "PreviewChannel" 1.0 :!class /Script/Engine.MaterialInstanceDynamic))
    (:NewEnumerator2
      _)
    (:NewEnumerator3
      _)
    (:NewEnumerator4
      _))
  (return))

(fn InitDebugMaterialTextures
  (params (MID :type Material%20Instance%20Dynamic%20Object%20Reference) (DebugMID :type Material%20Instance%20Dynamic%20Object%20Reference) (Params :type Array%20of%20Names) (IsNormal :type Boolean))
  (for _array_element Params
    (bind _returnvalue (Rendering|Material|GetTextureParameterValue MID _array_element :!class /Script/Engine.MaterialInstanceDynamic))
    (Utilities|IsValid _returnvalue
      (:Is%20Valid
        (Rendering|Material|SetTextureParameterValue DebugMID (select IsNormal "Normal" "Texture") _returnvalue)
        (Utilities|Array|AddUnique (Variables|Default|GetDebugCurrentTextures) _returnvalue)
        (return)))))

(fn InitMIDs
  (Utilities|Array|Clear (Variables|Hidden|GetMIDs))
  (for _array_element (Actor|GetComponentsByClass "/Script/Engine.SkeletalMeshComponent")
    (Utilities|Array|Clear (LocalVariables|Getl_MIDs))
    (Utilities|Array|Clear (LocalVariables|Getl_DebugMIDs))
    (for (_array_element_2 _array_index) (Rendering|Material|GetMaterials _array_element)
      (bind _returnvalue (Rendering|Material|CreateDynamicMaterialInstance _array_element_2 "None" "None" :!class /Script/Engine.KismetMaterialLibrary))
      (Utilities|Array|Add (LocalVariables|Getl_MIDs) _returnvalue)
      (Rendering|Material|SetMaterial _array_element _array_index _returnvalue)
      (bind _returnvalue_2 (Rendering|Material|CreateDynamicMaterialInstance (Variables|Hidden|GetDebugMaterial) "None" "None" :!class /Script/Engine.KismetMaterialLibrary))
      (Utilities|Array|Add (LocalVariables|Getl_DebugMIDs) _returnvalue_2))
    (Utilities|Map|Add (Variables|Hidden|GetMaterialMap) _array_element (Utilities|Struct|MakeMannequinHelperMIDData (LocalVariables|Getl_MIDs) (LocalVariables|Getl_DebugMIDs)))))

(fn SetUpDebugMaterialScalar
  (params (Vertex%20Color :type Boolean) (UV0 :type Boolean) (UV1 :type Boolean))
  (LocalVariables|Setcounter 0)
  (bind _keys (Utilities|Map|Keys (Variables|Hidden|GetMaterialMap)))
  (for _array_element _keys
    (bind _value (Utilities|Map|Find (Variables|Hidden|GetMaterialMap) _array_element))
    (bind (_value_mids_3_6ff767fe41164d399b8e9a9d765c88d3 _value_debugmids_5_ee2cbf9841eff7ba62c7a3b272b05e8d) (Utilities|Struct|BreakMannequinHelperMIDData _value))
    (for (_array_element_2 _array_index) _value_debugmids_5_ee2cbf9841eff7ba62c7a3b272b05e8d
      (Rendering|Material|SetScalarParameterValue _array_element_2 "VertexColor" (Math|Conversions|ToFloat(Boolean) Vertex%20Color) :!class /Script/Engine.MaterialInstanceDynamic)
      (Rendering|Material|SetScalarParameterValue _array_element_2 "UV0" (Math|Conversions|ToFloat(Boolean) UV0) :!class /Script/Engine.MaterialInstanceDynamic)
      (Rendering|Material|SetScalarParameterValue _array_element_2 "UV1" (Math|Conversions|ToFloat(Boolean) UV1) :!class /Script/Engine.MaterialInstanceDynamic)
      (Rendering|Material|SetScalarParameterValue _array_element_2 "PreviewUVs" (Math|Conversions|ToFloat(Boolean) (Variables|Default|GetDebugPreviewUvs)) :!class /Script/Engine.MaterialInstanceDynamic)
      (Rendering|Material|SetScalarParameterValue _array_element_2 "PreviewUVOffset" (Math|Conversions|ToFloat(Integer) (* (LocalVariables|Getcounter) 110)) :!class /Script/Engine.MaterialInstanceDynamic)
      (Rendering|Material|SetVectorParameterValue _array_element_2 "Channel" (Utilities|Select :Index (Variables|Default|GetDebugMaterialChannel) :NewEnumerator0 "(R=1.000000,G=0.000000,B=0.000000,A=0.000000)" :NewEnumerator1 "(R=1.000000,G=0.000000,B=0.000000,A=0.000000)" :NewEnumerator2 "(R=0.000000,G=1.000000,B=0.000000,A=0.000000)" :NewEnumerator3 "(R=0.000000,G=0.000000,B=1.000000,A=0.000000)" :NewEnumerator4 "(R=0.000000,G=0.000000,B=0.000000,A=1.000000)") :!class /Script/Engine.MaterialInstanceDynamic)
      (switch Utilities|FlowControl|Switch|SwitchonMannequinHelperChannels (Variables|Default|GetDebugMaterialChannel)
        (:NewEnumerator0
          (Rendering|Material|SetScalarParameterValue _array_element_2 "PreviewChannel" 0.0 :!class /Script/Engine.MaterialInstanceDynamic))
        (:NewEnumerator1
          (Rendering|Material|SetScalarParameterValue _array_element_2 "PreviewChannel" 1.0 :!class /Script/Engine.MaterialInstanceDynamic))
        (:NewEnumerator2
          _)
        (:NewEnumerator3
          _)
        (:NewEnumerator4
          _))
      (Rendering|Material|SetMaterial _array_element _array_index _array_element_2)
      (Math|Integer|IncrementInt (LocalVariables|Getcounter)))))

(fn ConstructionScript
  (Utilities|FlowControl|Sequence
    (:then_0
      (CallFunction|InitAllMaterials))
    (:then_1
      (for _array_element (Actor|GetComponentsByClass "/Script/Engine.SkeletalMeshComponent")
        (if (Variables|Default|GetDebugPreviewUvs)
          (Rendering|SetBoundsScale _array_element 1e+02)
          (else
            (Rendering|SetBoundsScale _array_element 1.0)))))))
