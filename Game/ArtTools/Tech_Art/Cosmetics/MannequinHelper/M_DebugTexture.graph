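# Unlit debug material for skeletal meshes that previews a texture, tangent or world-space normals, vertex color, or UV0/UV1 as flat color.
# Scalar parameters drive a chain of lerps, with optional channel isolation by dot product and a 2.2 gamma curve.
# Dividing by EyeAdaptation cancels auto exposure.
# A PreviewUVs branch computes position offsets that flatten the mesh into its UV layout at the actor position, spaced by island index and buffer.
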Material /Game/ArtTools/Tech_Art/Cosmetics/MannequinHelper/M_DebugTexture
graph {
 properties {bUsedWithMeshDeformer=true, bUsedWithSkeletalMesh=true, pixelDepthOffsetMode=PDOM_Legacy, shadingModel=MSM_Unlit, usedShadingModels=Unlit}
 nodes [
  n1 = TextureSampleParameter2D() {parameterName=Normal, samplerType=SAMPLERTYPE_Normal, texture=/Engine/EngineMaterials/FlatNormal}
  n2 = AppendVector(A=Transform(None=@n1.RGB), B=Constant())
  n3 = ScalarParameter() {parameterName=WorldSpaceNormal}
  n4 = LinearInterpolate(
   A=TextureSampleParameter2D() {parameterName=Texture, samplerType=SAMPLERTYPE_Masks, texture=&T_BlackMask}.RGBA
   B=ConstantBiasScale(None=LinearInterpolate(A=@n1.RGBA, B=@n2, Alpha=@n3))
   Alpha=ScalarParameter() {parameterName=IsNormal}
  )
  n5 = VertexColor()
  n6 = AppendVector(A=@n5, B=@n5)
  n7 = ScalarParameter() {parameterName=VertexColor}
  n8 = TextureCoordinate()
  n9 = Constant2Vector()
  n10 = LinearInterpolate(
   A=LinearInterpolate(A=LinearInterpolate(A=@n4, B=@n6, Alpha=@n7), B=AppendVector(A=@n8, B=@n9), Alpha=ScalarParameter() {parameterName=UV0})
   B=AppendVector(A=TextureCoordinate(), B=Constant2Vector())
   Alpha=ScalarParameter() {parameterName=UV1}
  )
  n11 = LinearInterpolate(
   A=@n10
   B=DotProduct(A=@n10, B=VectorParameter() {parameterName=Channel, defaultValue=<rgba 1 0 0 0>}.RGB)
   Alpha=ScalarParameter() {parameterName=PreviewChannel}
  )
  Divide(
   A=LinearInterpolate(A=@n11, B=Power(Base=@n11) {constExponent=2.2}, Alpha=ScalarParameter() {parameterName=DisplaySRGB})
   B=EyeAdaptation().EyeAdaptation
  )
  n12 = ActorPositionWS()
  n13 = ScalarParameter() {parameterName=PreviewUVChannel}
  n14 = ScalarParameter() {defaultValue=100, parameterName=PreviewUVIslandSize}
  n15 = Multiply(A=LinearInterpolate(A=TextureCoordinate(), B=TextureCoordinate() {coordinateIndex=1}, Alpha=Saturate(None=@n13)), B=@n14) {constB=100}
  n16 = Multiply(A=ComponentMask(None=@n15) {g=true}) {constB=-1}
  n17 = Multiply(A=@n14, B=ScalarParameter() {parameterName=PreviewUVIslandIndex})
  n18 = Multiply(A=@n14, B=ScalarParameter() {defaultValue=0.1, parameterName=PreviewUVIslandBuffer})
  n19 = Constant()
  n20 = Add(
   A=MaterialFunctionCall(X=ComponentMask(None=@n15) {r=true}, Y=Constant(), Z=Add(A=@n14, B=@n16) {constA=100}) {materialFunction=&MakeFloat3}.Result
   B=MaterialFunctionCall(X=Add(A=@n17, B=@n18), Y=@n19, Z=@n19) {materialFunction=&MakeFloat3}.Result
  )
  LinearInterpolate(B=Subtract(A=Add(A=@n12, B=@n20), B=WorldPosition().XYZ), Alpha=ScalarParameter() {parameterName=PreviewUVs}) {constAlpha=0}
 ]
}
