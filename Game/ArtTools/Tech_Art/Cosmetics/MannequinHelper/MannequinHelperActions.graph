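# Editor utility that converts a selected Fort player mannequin into a transient BP_MannequinHelper or separate SkeletalMeshActors.
# It caches each character part's mesh, then builds one skeletal mesh component per part.
# Part material overrides are applied by slot index, then variant swaps from the first non-wrap variant yielding materials, read through CosmeticPicker Python.
# The helper path also passes final materials to the helper and initializes its MIDs; both paths select the result.
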EditorUtilityBlueprint /Game/ArtTools/Tech_Art/Cosmetics/MannequinHelper/MannequinHelperActions
graphs [
 ConvertToMannequinHelper {nodes [
   ConvertToMannequinHelper() {is=title K2Node_FunctionEntry}
   > InitMannequinData()
   > CreateMannequinHelper()
   > OverrideCharacterPartMaterials()
   > OverrideVariantMaterials()
   > InitMannequinMaterialData()
   > InitMIDs(self:BP Mannequin Helper Object Reference=Get MyMannequinHelper() -> (MyMannequinHelper:BP Mannequin Helper Object Reference))
   > SetSelectedLevelActors(
    self=Get EditorActorSubsystem()
    ActorsToSelect=Make Array("[0]":Actor Object Reference=Get MyMannequinHelper()) -> (Array:Array of Actor Object References)
   )
  ]}
 ConvertToSkelMeshes {nodes [
   ConvertToSkelMeshes() {is=title K2Node_FunctionEntry}
   > InitMannequinData()
   > CreateSkelMeshes()
   > OverrideCharacterPartMaterials()
   > OverrideVariantMaterials()
   > SetSelectedLevelActors(self=Get EditorActorSubsystem(), ActorsToSelect=Get SkelMeshActors() -> (SkelMeshActors:Array of Actor Object References))
  ]}
 CreateMannequinHelper {nodes [
   CreateMannequinHelper() {is=title K2Node_FunctionEntry}
   n1 = Get SkeletalMeshComponent!(self:Skeletal Mesh Actor Object Reference=Get FortPlayerMannequin() -> (FortPlayerMannequin:Fort Player Mannequin Object Reference)) -> (SkeletalMeshComponent:Skeletal Mesh Component Object Reference)
   > n2 = SpawnActorFromClass!(
    execute
    self:Editor Actor Subsystem Object Reference=Get EditorActorSubsystem()
    ActorClass:Actor Class Reference=&/BP_MannequinHelper.BP_MannequinHelper_C
    Location:Vector=Get World Location!(self:Scene Component Object Reference=@n1) -> (ReturnValue:Vector) {is=title K2Node_CallFunction}
    Rotation:Rotator=Get World Rotation!(self:Scene Component Object Reference=@n1) -> (ReturnValue:Rotator) {is=title K2Node_CallFunction}
    bTransient:bool=true
   ) -> (then, ReturnValue:BP Mannequin Helper Object Reference)
   > Set MyMannequinHelper(MyMannequinHelper:BP Mannequin Helper Object Reference=@n2) -> (Output_Get:BP Mannequin Helper Object Reference)
   > Clear(TargetArray:Array of Skeletal Mesh Component Object References=Get SkelMeshComponents() -> (SkelMeshComponents:Array of Skeletal Mesh Component Object References))
   > n3 = For Each Loop!(
    Exec
    Array:Array of Mannequin Helper Data Struct Structures=Get MannequinData() -> (MannequinData:Array of Mannequin Helper Data Struct Structures)
   ) -> (
    LoopBody
    Array Element_PartType_5_#1:String
    Array Element_CharacterPart_14_#2:Custom Character Part Object Reference
    Array Element_SkeletalMesh_15_#3:Skeletal Mesh Object Reference
    Array Element_Materials_17_#4:Array of Material Interface Object References
    Array Index
    Completed
   )
   n4 = Get SubobjectDataSubsystem()
   > n5 = Gather Subobject Data For Instance(self=@n4, Context=Get MyMannequinHelper())
   > n6 = AddNewSubobject!(
    execute
    self:Subobject Data Subsystem Object Reference=@n4
    Params_ParentHandle:Subobject Data Handle Structure="Get (a ref)"(Array:Array of Subobject Data Handle Structures=@n5, Dimension 1=0) -> (Output:"Subobject Data Handle Structure (by ref)")
    Params_NewClass:Object Class Reference=/Script/Engine.SkeletalMeshComponent
    Params_BlueprintContext:Blueprint Object Reference
    Params_bSkipMarkBlueprintModified:bool=False
    Params_bConformTransformToParent:bool=True
   ) -> (then, FailReason:Text, ReturnValue:Subobject Data Handle Structure)
   n7 = GetPartType!(self:Custom Character Part Object Reference=@n3.Array Element_CharacterPart_14_#2) -> (ReturnValue:EFortCustomPartType Enum=Head)
   > RenameSubobject!(
    execute
    self:Subobject Data Subsystem Object Reference=@n4
    Handle:"Subobject Data Handle Structure (by ref)"=@n6.ReturnValue
    InNewName:"Text (by ref)"="To Text (String)"(InString=Enum to String!(Enumerator:EFortCustomPartType Enum=@n7) -> (ReturnValue:String))
   ) -> (then, ReturnValue:bool)
   > n8 = Cast To SkeletalMeshComponent!(
    execute
    Object=GetObject!(Data:"Subobject Data Structure (by ref)"=GetData(DataHandle=@n6.ReturnValue), bEvenIfPendingKill:bool=false) -> (ReturnValue:Object Reference)
   ) -> (then, CastFailed, AsSkeletal Mesh Component:Skeletal Mesh Component Object Reference) {is=title K2Node_DynamicCast}
   > SetSkeletalMeshAsset!(
    execute
    self:Skeletal Mesh Component Object Reference=@n8
    NewMesh:Skeletal Mesh Object Reference=@n3.Array Element_SkeletalMesh_15_#3
   ) -> (then)
   > Add(
    TargetArray:Array of Skeletal Mesh Component Object References=Get SkelMeshComponents()
    NewItem:"Skeletal Mesh Component Object Reference (by ref)"=@n8
   ) {is=title K2Node_CallArrayFunction}
  ]}
 CreateSkelMeshes {nodes [
   CreateSkelMeshes() {is=title K2Node_FunctionEntry}
   > Clear(TargetArray=Get SkelMeshComponents())
   > Clear(TargetArray:Array of Actor Object References=Get SkelMeshActors())
   > n1 = For Each Loop(Array=Get MannequinData())
   n2 = Get SkeletalMeshComponent(self=Get FortPlayerMannequin())
   > n3 = SpawnActorFromClass(
    self=Get EditorActorSubsystem()
    ActorClass=/Script/Engine.SkeletalMeshActor
    Location=Get World Location(self=@n2)
    Rotation=Get World Rotation(self=@n2)
   ) -> (ReturnValue:Skeletal Mesh Actor Object Reference)
   n4 = GetAthenaCharacterItemDefinition!(self:Fort Player Mannequin Object Reference=Get FortPlayerMannequin()) -> (ReturnValue:Athena Character Item Definition Object Reference)
   n5 = GetPartType(self=@n1.Array Element_CharacterPart_14_#2) -> (ReturnValue=Head)
   > SetActorLabel(
    self=@n3
    NewActorLabel=Append!(A=GetDisplayName(Object=@n4), B=_, C:String=Enum to String(Enumerator=@n5)) -> (ReturnValue)
    bMarkDirty=true
   )
   > Add(TargetArray:Array of Actor Object References=Get SkelMeshActors(), NewItem:"Actor Object Reference (by ref)"=@n3)
   n6 = Get SkeletalMeshComponent(self=@n3)
   > SetSkeletalMeshAsset(self=@n6, NewMesh=@n1.Array Element_SkeletalMesh_15_#3)
   > Add(
    TargetArray:Array of Skeletal Mesh Component Object References=Get SkelMeshComponents()
    NewItem:"Skeletal Mesh Component Object Reference (by ref)"=@n6
   )
  ]}
 EventGraph {links [], nodes []}
 GetMannequin {nodes [
   n1 = GetMannequin!() -> (then, SelectionSet:Array of Actor Object References) {is=title K2Node_FunctionEntry}
   > Branch(Condition="integer > integer"(A=Length(TargetArray=@n1)))
   n2 = "Get (a copy)"(Array:Array of Actor Object References=@n1) -> (Output:Actor Object Reference)
   > Is Valid(InputObject=@n2)
   > n3 = Cast To FortPlayerMannequin!(execute, Object=@n2) -> (then, CastFailed, AsFort Player Mannequin:Fort Player Mannequin Object Reference) {is=title K2Node_DynamicCast}
   > Return Node!(execute, Mannequin:Fort Player Mannequin Object Reference=@n3)
   FortShippingLog!(execute=@n3.CastFailed, WorldContextObject:Object Reference, InString:String=Selected Actor Is Not A Mannequin, bLogAsWarning:bool=false) -> (then)
  ]}
 GetMannequinVariantMaterials {nodes [
   GetMannequinVariantMaterials() {is=title K2Node_FunctionEntry}
   > Clear(TargetArray:Array of Material Interface Object References=Get swapMats() -> (swapMats:Array of Material Interface Object References))
   n1 = GetAthenaCharacterItemDefinition(self=Get FortPlayerMannequin())
   > Set characterID(characterID:Athena Character Item Definition Object Reference=@n1) -> (Output_Get:Athena Character Item Definition Object Reference)
   > n2 = Execute Python Script!(
    execute
    PythonScript="import TechArt.CosmeticPicker.CosmeticPickerModule as CosmeticPicker\r\nnum = CosmeticPicker.getNumVariants(id)"
    id:Athena Character Item Definition Object Reference=@n1
   ) -> (then, ReturnValue, num:int) {inputs [id], outputs [num]}
   > n3 = For Loop(LastIndex=int - int(A=@n2.num))
   > n4 = Set curIndex(curIndex:int=@n3) -> (Output_Get:int=0)
   > n5 = Execute Python Script!(
    execute
    PythonScript="import TechArt.CosmeticPicker.CosmeticPickerModule as CosmeticPicker\r\n\r\ntype = CosmeticPicker.getVariantType(id, variant_index)"
    id=Get characterID() -> (characterID:Athena Character Item Definition Object Reference)
    variant_index:int=@n4
   ) -> (then, ReturnValue, type:String) {inputs [id, variant_index], outputs [type]}
   > n6 = Branch(Condition="Equal Exactly (String)"(A=@n5.type, B=FortCosmeticVariantApplyWrap))
   n7 = Execute Python Script!(
    execute=@n6.else
    PythonScript="import TechArt.CosmeticPicker.CosmeticPickerModule as CosmeticPicker\r\n\r\ndefault_index = CosmeticPicker.getDefaultVariantDefaultOption(id, variant_index)\r\nactive_index = CosmeticPicker.checkMannequinForComboboxVariant(mannequin, id, variant_index, default_index)"
    mannequin:Fort Player Mannequin Object Reference=Get FortPlayerMannequin()
    id=Get characterID()
    variant_index=Get curIndex() -> (curIndex:int=0)
   ) -> (then, ReturnValue, default_index:int, active_index:Integer64) {inputs [mannequin, id, variant_index], outputs [default_index, active_index]}
   > Set defaultIndex(defaultIndex:int=@n7.default_index) -> (Output_Get:int=0)
   > n8 = Execute Python Script!(
    execute
    PythonScript="import TechArt.CosmeticPicker.CosmeticPickerModule as CosmeticPicker\r\n\r\nto_swap_materials = []\r\noverride_materials = []\r\n\r\nitemVariants = id.get_editor_property(\"itemVariants\")\r\nvariant_options = CosmeticPicker.getOptionsParam(itemVariants[variant_index])\r\nif active_index < len(variant_options):\r\n    variant_option = variant_options[active_index]\r\n    variant_materials = variant_option.get_editor_property(\"variantMaterials\")\r\n    \r\n    for m in variant_materials:\r\n        to_swap_materials.append(m.get_editor_property(\"MaterialtoSwap\"))\r\n        override_materials.append(m.get_editor_property(\"OverrideMaterial\"))"
    id=Get characterID()
    variant_index=Get curIndex() -> (curIndex=0)
    active_index:int="To Integer (Integer64)"!(InInt:Integer64=@n7.active_index) -> (ReturnValue:int=0)
   ) -> (
    then
    ReturnValue
    to_swap_materials:Array of Material Interface Object References
    override_materials:Array of Material Interface Object References
   ) {inputs [id, variant_index, active_index], outputs [to_swap_materials, override_materials]}
   > Branch(Condition="integer > integer"(A=Length(TargetArray:Array of Material Interface Object References=@n8.to_swap_materials)))
   > n9 = Set swapMats(swapMats:Array of Material Interface Object References=@n8.to_swap_materials) -> (Output_Get:Array of Material Interface Object References)
   > n10 = Set overrideMats(overrideMats:Array of Material Interface Object References=@n8.override_materials) -> (Output_Get:Array of Material Interface Object References)
   > Return Node!(
    execute
    SwapMaterials:Array of Material Interface Object References=@n9
    OverrideMaterials:Array of Material Interface Object References=@n10
   )
  ]}
 InitMannequinData {nodes [
   InitMannequinData() {is=title K2Node_FunctionEntry}
   > n1 = GetSelectionSet!(execute) -> (then, ReturnValue:Array of Actor Object References)
   > n2 = GetMannequin!(execute, self, SelectionSet:Array of Actor Object References=@n1) -> (then, Mannequin:Fort Player Mannequin Object Reference) {is=title K2Node_CallFunction}
   > Is Valid(InputObject=@n2)
   > n3 = Set FortPlayerMannequin(FortPlayerMannequin:Fort Player Mannequin Object Reference=@n2) -> (Output_Get:Fort Player Mannequin Object Reference)
   > n4 = GetCustomCharacterParts!(execute, self:Fort Player Mannequin Object Reference=@n3) -> (then, ReturnValue:Array of Custom Character Part Object References)
   > n5 = For Each Loop!(Exec, Array:Array of Custom Character Part Object References=@n4) -> (LoopBody, Array Element:Custom Character Part Object Reference, Array Index, Completed)
   Return Node!(execute=@n5.Completed)
   > Set curCharacterPart(curCharacterPart:Custom Character Part Object Reference=@n5) -> (Output_Get:Custom Character Part Object Reference)
   n6 = Get curCharacterPart() -> (curCharacterPart:Custom Character Part Object Reference)
   n7 = GetPartType(self=@n6) -> (ReturnValue=Head)
   > Add(
    TargetArray:Array of Mannequin Helper Data Struct Structures=Get MannequinData()
    NewItem:"Mannequin Helper Data Struct Structure (by ref)"=Make Mannequin Helper Data Struct!(
     PartType_5_#1:String=Enum to String(Enumerator=@n7)
     CharacterPart_14_#2:Custom Character Part Object Reference=@n6
     SkeletalMesh_15_#3:Skeletal Mesh Object Reference=GetSkeletalMesh!(self:Custom Character Part Object Reference=@n6) -> (ReturnValue:Skeletal Mesh Object Reference)
     Materials_17_#4:Array of Material Interface Object References
    ) -> (MannequinHelperDataStruct:Mannequin Helper Data Struct Structure) {is=title K2Node_MakeStruct}
   )
  ]}
 InitMannequinMaterialData {nodes [
   InitMannequinMaterialData() {is=title K2Node_FunctionEntry}
   > n1 = For Each Loop(Array:Array of Skeletal Mesh Component Object References=Get SkelMeshComponents()) -> (Array Element:Skeletal Mesh Component Object Reference)
   Set MannequinData!(
    execute=@n1.Completed
    MannequinData:Array of Mannequin Helper Data Struct Structures=Get MannequinData()
    self:BP Mannequin Helper Object Reference=Get MyMannequinHelper()
   ) -> (then, Output_Get:Array of Mannequin Helper Data Struct Structures)
   Set members in Mannequin Helper Data Struct!(
    execute:Exec=@n1.LoopBody
    StructRef:"Mannequin Helper Data Struct Structure (by ref)"="Get (a ref)"(Array:Array of Mannequin Helper Data Struct Structures=Get MannequinData(), Dimension 1=@n1.Array Index) -> (Output:"Mannequin Helper Data Struct Structure (by ref)")
    Materials_17_#4:Array of Material Interface Object References=GetMaterials!(self:Mesh Component Object Reference=@n1) -> (ReturnValue:Array of Material Interface Object References)
   ) -> (then:Exec, StructOut:"Mannequin Helper Data Struct Structure (by ref)") {is=title K2Node_SetFieldsInStruct}
  ]}
 OverrideCharacterPartMaterials {nodes [
   OverrideCharacterPartMaterials() {is=title K2Node_FunctionEntry}
   > Clear(TargetArray=Get mats() -> (mats:Array of Material Interface Object References))
   > n1 = For Each Loop!(Exec, Array:Array of Mannequin Helper Data Struct Structures=Get MannequinData()) -> (
    LoopBody
    Array Element_PartType_5_#1:String
    Array Element_CharacterPart_14_#2:Custom Character Part Object Reference
    Array Element_SkeletalMesh_15_#3:Skeletal Mesh Object Reference
    Array Element_Materials_17_#4:Array of Material Interface Object References
    Array Index
    Completed
   )
   n2 = "Get (a copy)"(Array:Array of Skeletal Mesh Component Object References=Get SkelMeshComponents(), Dimension 1=@n1.Array Index) -> (Output:Skeletal Mesh Component Object Reference)
   > n3 = For Each Loop!(Exec, Array:Array of Material Interface Object References=GetMaterials(self=@n2)) -> (LoopBody, Array Element:Material Interface Object Reference, Array Index, Completed)
   n4 = Find(
    TargetMap:Map of Integers to Material Interface Object References=GetMaterialOverridesByIndex!(self:Custom Character Part Object Reference=@n1.Array Element_CharacterPart_14_#2) -> (ReturnValue:Map of Integers to Material Interface Object References)
    Key:"Integer (by ref)"=@n3.Array Index
   ) -> (Value:Material Interface Object Reference)
   > Branch(Condition=@n4.ReturnValue)
   > SetMaterial(self=@n2, ElementIndex=@n3.Array Index, Material=@n4)
  ]}
 OverrideVariantMaterials {nodes [
   OverrideVariantMaterials() {is=title K2Node_FunctionEntry}
   > n1 = GetMannequinVariantMaterials!(execute, self) -> (then, SwapMaterials:Array of Material Interface Object References, OverrideMaterials:Array of Material Interface Object References) {is=title K2Node_CallFunction}
   > Set swapMats(swapMats=@n1)
   > Set overrideMats(overrideMats=@n1.OverrideMaterials)
   > n2 = For Each Loop(Array:Array of Skeletal Mesh Component Object References=Get SkelMeshComponents()) -> (Array Element:Skeletal Mesh Component Object Reference)
   > n3 = Set curSkelMesh(curSkelMesh:Skeletal Mesh Component Object Reference=@n2) -> (Output_Get:Skeletal Mesh Component Object Reference)
   > n4 = For Each Loop(Array:Array of Material Interface Object References=GetMaterials(self=@n3)) -> (Array Element:Material Interface Object Reference)
   n5 = Find Item!(TargetArray:Array of Material Interface Object References=Get swapMats(), ItemToFind:"Material Interface Object Reference (by ref)"=@n4) -> (ReturnValue)
   > Branch(Condition="integer >= integer"(A=@n5))
   > SetMaterial(
    self=Get curSkelMesh() -> (curSkelMesh:Skeletal Mesh Component Object Reference)
    ElementIndex=@n4.Array Index
    Material="Get (a copy)"(
     Array:Array of Material Interface Object References=Get overrideMats() -> (overrideMats:Array of Material Interface Object References)
     Dimension 1=@n5
    ) -> (Output:Material Interface Object Reference)
   )
  ]}
]
variables [
 DebugMaterial {type=MaterialInterface, default=&/M_DebugTexture}
 FortPlayerMannequin {type=FortPlayerMannequin, default=null}
 MannequinData {type="array<MannequinHelperDataStruct>", default []}
 MyMannequinHelper {type=BP_MannequinHelper_C, default=null}
 SkelMeshActors {type="array<Actor>"}
 SkelMeshComponents {type="array<SkeletalMeshComponent>", default []}
]
